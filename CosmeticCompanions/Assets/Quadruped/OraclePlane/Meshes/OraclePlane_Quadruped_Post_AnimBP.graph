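# Post-process AnimBP for the OraclePlane quadruped cosmetic companion, adding secondary motion to its tail, hair and hat.
# A PoseAsset is applied to the input pose, then chained spring-driven Anim Dynamics nodes simulate tail_02 to tail_06 relative to their bones, and HairVolume and hat relative to root.
# Alphas are inverted (1 minus input): DynamicsAlpha fades out tail and hair dynamics, DisableHatDyn the hat's.
# The EventGraph is empty.
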AnimBlueprint /CosmeticCompanions/Assets/Quadruped/OraclePlane/Meshes/OraclePlane_Quadruped_Post_AnimBP
target_skeleton=&/OraclePlane_Quadruped_Skeleton
graphs [
 AnimGraph {nodes [
   n1 = Input Pose()
   n2 = Get PoseAsset() -> (PoseAsset:Pose Asset Object Reference)
   n3 = Get DynamicsAlpha() -> (DynamicsAlpha:float)
   n4 = Anim Dynamics - Bone: tail_02(
    ComponentPose=Local To Component(LocalPose=Pose!(PoseAsset:Pose Asset Object Reference=@n2, SourcePose:Pose Link Structure=@n1) -> (Pose:Pose Link Structure) {is=title AnimGraphNode_PoseBlendNode, blendOption=Linear})
    Alpha=@n3
    ExternalForce="0,0,0"
   ) {
    angularDampingOverride=1
    angularSpringConstant=15000
    bAngularSpring=true
    bOverrideAngularDamping=true
    bUseGravityOverride=true
    gravityScale=0.35
    simulationSpace=BoneRelative
    alphaBoolBlend {blendOption=Linear}
    alphaScaleBiasClamp {bias=1, scale=-1}
    boundBone {boneName=tail_02}
    componentLinearAccScale=<1 1 1>
    componentLinearVelScale=<1 1 1>
    gravityOverride {z=-980}
    physicsBodyDefinitions [{
      boundBone {boneName=tail_02}
      boxExtents=<20 30 30>
      constraintSetup {angularLimitsMax=<5 3 15>, angularLimitsMin=<-5 -45 -15>, angularTarget=<1 0 0>}
      localJointOffset=<4 0 0>
     }]
    relativeSpaceBone {boneName=hips}
   }
   n5 = Anim Dynamics - Bone: tail_03(ComponentPose=@n4, Alpha=@n3, ExternalForce="0,0,0") {
    angularDampingOverride=1
    angularSpringConstant=15000
    bAngularSpring=true
    bOverrideAngularDamping=true
    bUseGravityOverride=true
    gravityScale=0.35
    simulationSpace=BoneRelative
    alphaBoolBlend {blendOption=Linear}
    alphaScaleBiasClamp {bias=1, scale=-1}
    boundBone {boneName=tail_03}
    componentLinearAccScale=<1 1 1>
    componentLinearVelScale=<1 1 1>
    gravityOverride {z=-980}
    physicsBodyDefinitions [{
      boundBone {boneName=tail_03}
      boxExtents=<20 30 30>
      constraintSetup {angularLimitsMax=<5 3 15>, angularLimitsMin=<-5 -45 -15>, angularTarget=<1 0 0>}
      localJointOffset=<4 0 0>
     }]
    relativeSpaceBone {boneName=hips}
   }
   n6 = Anim Dynamics - Bone: tail_04(ComponentPose=@n5, Alpha=@n3, ExternalForce="0,0,0") {
    angularDampingOverride=1
    angularSpringConstant=15000
    bAngularSpring=true
    bOverrideAngularDamping=true
    bUseGravityOverride=true
    gravityScale=0.35
    simulationSpace=BoneRelative
    alphaBoolBlend {blendOption=Linear}
    alphaScaleBiasClamp {bias=1, scale=-1}
    boundBone {boneName=tail_04}
    componentLinearAccScale=<1 1 1>
    componentLinearVelScale=<1 1 1>
    gravityOverride {z=-980}
    physicsBodyDefinitions [{
      boundBone {boneName=tail_04}
      boxExtents=<20 30 30>
      constraintSetup {angularLimitsMax=<5 3 15>, angularLimitsMin=<-5 -45 -15>, angularTarget=<1 0 0>}
      localJointOffset=<4 0 0>
     }]
    relativeSpaceBone {boneName=hips}
   }
   n7 = Anim Dynamics - Bone: tail_05(ComponentPose=@n6, Alpha=@n3, ExternalForce="0,0,0") {
    angularDampingOverride=1
    angularSpringConstant=15000
    bAngularSpring=true
    bOverrideAngularDamping=true
    bUseGravityOverride=true
    gravityScale=0.35
    simulationSpace=BoneRelative
    alphaBoolBlend {blendOption=Linear}
    alphaScaleBiasClamp {bias=1, scale=-1}
    boundBone {boneName=tail_05}
    componentLinearAccScale=<1 1 1>
    componentLinearVelScale=<1 1 1>
    gravityOverride {z=-980}
    physicsBodyDefinitions [{
      boundBone {boneName=tail_05}
      boxExtents=<20 30 30>
      constraintSetup {angularLimitsMax=<5 3 15>, angularLimitsMin=<-5 -45 -15>, angularTarget=<1 0 0>}
      localJointOffset=<4 0 0>
     }]
    relativeSpaceBone {boneName=hips}
   }
   n8 = Anim Dynamics - Bone: tail_06(ComponentPose=@n7, Alpha=@n3, ExternalForce="0,0,0") {
    angularDampingOverride=1
    angularSpringConstant=15000
    bAngularSpring=true
    bOverrideAngularDamping=true
    bUseGravityOverride=true
    gravityScale=0.35
    simulationSpace=BoneRelative
    alphaBoolBlend {blendOption=Linear}
    alphaScaleBiasClamp {bias=1, scale=-1}
    boundBone {boneName=tail_06}
    componentLinearAccScale=<1 1 1>
    componentLinearVelScale=<1 1 1>
    gravityOverride {z=-980}
    physicsBodyDefinitions [{
      boundBone {boneName=tail_06}
      boxExtents=<20 30 30>
      constraintSetup {angularLimitsMax=<5 3 15>, angularLimitsMin=<-5 -45 -15>, angularTarget=<1 0 0>}
      localJointOffset=<4 0 0>
     }]
    relativeSpaceBone {boneName=hips}
   }
   n9 = Anim Dynamics - Bone: HairVolume(ComponentPose=@n8, Alpha=@n3, ExternalForce="0,0,0") {
    angularDampingOverride=0.9999
    angularSpringConstant=1500
    bAngularSpring=true
    bOverrideAngularDamping=true
    gravityScale=0.35
    simulationSpace=RootRelative
    alphaBoolBlend {blendOption=Linear}
    alphaScaleBiasClamp {bias=1, scale=-1}
    boundBone {boneName=HairVolume}
    componentAppliedLinearAccClamp=<3000 3000 3000>
    physicsBodyDefinitions [{
      boundBone {boneName=HairVolume}
      boxExtents=<10 10 10>
      constraintSetup {angularLimitsMax=<25 25 25>, angularLimitsMin=<-25 -25 -25>, angularTarget=<1 0 0>}
      localJointOffset=<8 0 -3>
     }]
    relativeSpaceBone {boneName=hips}
   }
   n10 = Anim Dynamics - Bone: hat(ComponentPose=@n9, Alpha=Get DisableHatDyn() -> (DisableHatDyn:double), ExternalForce="0,0,0") {
    angularDampingOverride=1
    angularSpringConstant=1500
    bAngularSpring=true
    bOverrideAngularDamping=true
    bUseGravityOverride=true
    gravityScale=0.35
    simulationSpace=RootRelative
    alphaBoolBlend {blendOption=Linear}
    alphaScaleBiasClamp {bias=1, scale=-1}
    boundBone {boneName=hat}
    componentAppliedLinearAccClamp=<3000 3000 3000>
    componentLinearAccScale=<1 1 1>
    componentLinearVelScale=<1 1 1>
    gravityOverride {z=-980}
    physicsBodyDefinitions [{
      boundBone {boneName=hat}
      boxExtents=<20 30 30>
      constraintSetup {angularLimitsMax=<3 2 1>, angularLimitsMin=<-3 -2 -1>, angularTarget=<1 0 0>}
      localJointOffset=<0 0 2>
     }]
    relativeSpaceBone {boneName=hips}
   }
   Output Pose(Result=Component To Local(ComponentPose=@n10))
  ]}
 EventGraph {links [], nodes []}
]
variables [
 DisableHatDyn {type=number}
 PoseAsset {type=PoseAsset, default=&/OraclePlane_Quadruped_PoseAsset}
]
